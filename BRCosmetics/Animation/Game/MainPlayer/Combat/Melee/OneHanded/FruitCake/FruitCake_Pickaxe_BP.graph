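# Animation Blueprint for the FruitCake pickaxe's weapon skeleton, supplying its idle pose and montage playback.
# A single-state machine plays the weapon idle sequence, then passes through a Constraint node and a Transform Bone node on Melee_TwoHanded_Handle in component space.
# The Transform node replaces that bone's rotation with a fixed offset, and DefaultSlot layers montages over the result before output.
# The event graph has no logic.
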AnimBlueprint /BRCosmetics/Animation/Game/MainPlayer/Combat/Melee/OneHanded/FruitCake/FruitCake_Pickaxe_BP
target_skeleton=/BRCosmetics/Weapons/FORT_Melee/Pickaxe_Fruitcake_Female/Meshes/Fruitcake_Female_Axe_Skeleton
graphs [
 AnimGraph {nodes [
   Local Space Ref Pose()
   AnimGraphNode_StateMachine_0 = New State Machine() {is=title AnimGraphNode_StateMachine}
   n1 = "Transform (Modify) Bone - Bone: Melee_TwoHanded_Handle"(
    ComponentPose="Constraint - Melee_TwoHanded_Handle ()"!(ComponentPose=Local To Component(LocalPose=@AnimGraphNode_StateMachine_0), Alpha=1) -> (Pose) {alphaBoolBlend {blendOption=Linear}, boneToModify {boneName=Melee_TwoHanded_Handle}}
    Alpha=1
    Translation="0,0,0"
    Rotation="0,-90,5"
    Scale="1,1,1"
   ) {rotationMode=BMM_Replace, alphaBoolBlend {blendOption=Linear}, boneToModify {boneName=Melee_TwoHanded_Handle}, rotation {roll=5, yaw=-90}}
   Output Pose(Result=Slot 'DefaultSlot'(Source=Component To Local(ComponentPose=@n1))) {layerGroup=None}
  ]}
 New State Machine {entry=Idle, path=AnimGraph.AnimGraphNode_StateMachine_0.New State Machine, ids=id, nodes [AnimStateNode_0 = Idle()]}
 Idle {
  path=AnimGraph.AnimGraphNode_StateMachine_0.New State Machine.AnimStateNode_0.Idle
  nodes [Output Animation Pose(Result=Sequence Player 'FruitCake_Idle_Pose_WEP'() {sequence=&/FruitCake_Idle_Pose_WEP})]
 }
 EventGraph {nodes [
   TryGetPawnOwner()
   Event BlueprintUpdateAnimation()
  ]}
]
variables []
